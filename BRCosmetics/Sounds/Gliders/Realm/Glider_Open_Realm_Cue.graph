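# Sound cue for opening the Realm glider, layering an open one-shot with a thrust sound.
# A Mixer plays the open wave immediately alongside the thrust loop wave, which starts after a fixed delay.
# An Enveloper fades the thrust layer in, holds it briefly, then fades it out; the root Random has only this one branch.
# It uses glider open/close attenuation and concurrency.
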SoundCue /BRCosmetics/Sounds/Gliders/Realm/Glider_Open_Realm_Cue
root=Random_0
nodes [
 {id=Delay_0, children [Enveloper_0], properties {delayMax=2, delayMin=2}}
 {
  id=Enveloper_0
  children [WavePlayer_1]
  properties {
   pitchCurve {editorCurveData {keys [{time=0}]}}
   volumeCurve {editorCurveData {keys [<at 0.0147 0.0196>,<at 0.691 0.9961>,<at 1.8511 0.9961>,<at 4.4502 -0.0358>]}}
  }
 }
 {id=Mixer_0, children [WavePlayer_0,Delay_0], properties {inputVolume [1,1]}}
 {id=Random_0, children [Mixer_0], properties {weights [1]}}
 {id=WavePlayer_0, properties {soundWaveAssetPtr=&/Glider_Realm_Open_01}}
 {id=WavePlayer_1, properties {soundWaveAssetPtr=&/Glider_Realm_Thrust_Loop_01}}
]
properties {
 attenuationSettings=&/Fort_Player_Glider_OpenClose_Realm_Athena_Att
 audioPropertiesSheet=/BRCosmetics/Sounds/Settings/PropertySheets/Glider/APS_Cosmetic_Glider_Open
 maxDistance=10000
 soundClassObject=&Fort_Gliders_Land
 volumeMultiplier=1.2
 concurrencySet [&Athena_GliderOpenClose_Conc]
}
